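# Editor utility widget that serves as a reusable labeled checkbox tile, apparently for the Environment BPSetup art tool.
# A small fixed-size rounded border holds a bold text label above the checkbox.
# UpdateData copies CheckboxValue into the label, while CheckboxKey is stored but not read here.
# Toggling the checkbox broadcasts OnCheckboxChanged with the checked state, the widget itself as sender, and the checked boolean.
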EditorUtilityWidgetBlueprint /Game/ArtTools/Environment/Implementation/BPSetup/EUW_BPSetup_Checkbox
graphs [
 EventGraph {nodes [
   n1 = "On Check State Changed (Checkbox)"!() -> (OutputDelegate, then, bIsChecked:bool) {is=title K2Node_ComponentBoundEvent}
   > Call On Checkbox Changed!(
    execute:Exec
    self:EUW BPSetup Checkbox Object Reference
    CheckboxState:ECheckBoxState Enum=GetCheckedState!(self:Check Box Object Reference=Get Checkbox() -> (Checkbox:Editor Utility Check Box Object Reference)) -> (ReturnValue:ECheckBoxState Enum=Unchecked)
    Sender:EUW BPSetup Checkbox Object Reference=Self-Reference()
    IsChecked:bool=@n1.bIsChecked
   ) -> (then:Exec) {is=title K2Node_CallDelegate}
   UpdateData() {is=title K2Node_CustomEvent}
   > Set Text(
    Text="To Text (String)"(InString=Get CheckboxValue() -> (CheckboxValue:String))
    self=Get Checkbox_Text() -> (Checkbox_Text:Text Object Reference)
   )
  ]}
 OnCheckboxChanged {nodes [OnCheckboxChanged!() -> (then, CheckboxState:ECheckBoxState Enum=Unchecked, Sender:EUW BPSetup Checkbox Object Reference, IsChecked:bool=~) {is=title K2Node_FunctionEntry}]}
]
variables [
 CheckboxKey {type=string, default=""}
 CheckboxValue {type=string, default=""}
 OnCheckboxChanged {default=()}
]
widgets {
 info {rootWidgetClass=&VerticalBox, widgetCount=6}
 widgets [
  VerticalBox_1 {widgetClassPath=&VerticalBox, widget=VerticalBox_1}
  SizeBox_1 {
   widgetClassPath=&SizeBox
   parent=VerticalBox_1
   widget=SizeBox_1
   properties {
    bOverride_HeightOverride=true
    bOverride_MaxDesiredHeight=true
    bOverride_MaxDesiredWidth=true
    bOverride_WidthOverride=true
    heightOverride=45
    maxDesiredHeight=45
    maxDesiredWidth=75
    widthOverride=75
   }
   slot {class=VerticalBoxSlot, properties {horizontalAlignment=HAlign_Center, verticalAlignment=VAlign_Center, padding=<ltrb 2 2 2 2>}}
  }
  Border_BPs {
   widgetClassPath=&Border
   parent=SizeBox_1
   widget=Border_BPs
   properties {
    background {
     bIsDynamicallyLoaded=false
     drawAs=RoundedBox
     imageType=NoImage
     mirroring=NoMirror
     resourceName=None
     resourceObject=null
     tiling=NoTile
     imageSize=<32 32>
     margin=<ltrb 0 0 0 0>
     outlineSettings {
      bUseBrushTransparency=false
      roundingType=FixedRadius
      width=1
      color {colorUseRule=UseColor_Specified, specifiedColor=<rgba 0.099 0.099 0.099 1>}
      cornerRadii=<5 5 5 5>
     }
     tintColor {colorUseRule=UseColor_Specified, specifiedColor=<rgba 0.0194 0.0194 0.0194 1>}
     uVRegion {bIsValid=false, max=<0 0>, min=<0 0>}
    }
    padding=<ltrb 2.5 2.5 2.5 2.5>
   }
   slot {class=SizeBoxSlot}
  }
  VerticalBox_0 {
   widgetClassPath=&VerticalBox
   parent=Border_BPs
   widget=VerticalBox_0
   slot {class=BorderSlot, properties {padding=<ltrb 2.5 2.5 2.5 2.5>}}
  }
  Checkbox_Text {
   widgetClassPath=&TextBlock
   bIsVariable=true
   parent=VerticalBox_0
   widget=Checkbox_Text
   properties {
    text=Straight
    font {
     bForceMonospaced=false
     bMaterialIsStencil=false
     fontMaterial=null
     fontObject=&Roboto
     letterSpacing=0
     monospacedWidth=1
     size=7.5
     skewAmount=0
     typefaceFontName=Bold
     outlineSettings {
      bApplyOutlineToDropShadows=false
      bMiteredCorners=false
      bSeparateFillAlpha=false
      outlineMaterial=null
      outlineSize=0
      outlineColor=<rgba 0 0 0 1>
     }
    }
   }
   slot {class=VerticalBoxSlot, properties {horizontalAlignment=HAlign_Center, verticalAlignment=VAlign_Center}}
  }
  Checkbox {
   widgetClassPath=&EditorUtilityCheckBox
   bIsVariable=true
   parent=VerticalBox_0
   widget=Checkbox
   slot {class=VerticalBoxSlot, properties {horizontalAlignment=HAlign_Center, verticalAlignment=VAlign_Center}}
  }
 ]
}
